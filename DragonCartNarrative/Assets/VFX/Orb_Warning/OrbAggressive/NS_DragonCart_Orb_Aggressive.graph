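# Niagara system for the aggressive look of the DragonCart warning orb: a hot red-orange shell with flare and electric arcs.
# Minimal animates twelve mesh particles in three groups, rotating and sliding them on looping emitter-age curves.
# FlareLight adds glow sprites and a light, while ElectricArcs spawns short-lived sprites on a ring.
# All follow System.OrbWPO; the sweeping TargetingBeam001, gated by User.Targeting, is disabled.
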NiagaraSystem /DragonCartNarrative/Assets/VFX/Orb_Warning/OrbAggressive/NS_DragonCart_Orb_Aggressive
dependencies {
 usedDataInterfaces [
  &NiagaraDataInterfaceMeshRendererInfo
  &NiagaraDataInterfaceCurve
  &NiagaraDataInterfaceArrayFloat
  &NiagaraDataInterfaceColorCurve
 ]
 usedDynamicInputs [
  &SimulationPosition
  &Vector2DFromFloat
  &self:ScratchDynamicInput_01
 ]
 usedModules [
  &SystemState
  &self:SystemSpawnScript.SystemScriptSource.SystemScriptGraph.NiagaraNodeAssignment_0.SetVariables_#1
 ]
}
emitters {
 ElectricArcs {
  emitterUpdateScript [
   &EmitterState(Life Cycle Mode=::NewEnumerator1, Loop Duration=2, UseLoopDelay=false)
   &SpawnRate(Spawn Probability=0.5, SpawnRate=5, Use Spawn Probability=true)
   &SpawnBurst_Instantaneous(
    Spawn Count=&RandomRangeInt(Minimum=4, Maximum=5)
    Use Spawn Probability=false
    UseLoopCountLimit=false
   )
   &SpawnBurst_Instantaneous(
    Spawn Count=&RandomRangeInt(Minimum=2, Maximum=4)
    Spawn Time=0.16
    Use Spawn Probability=false
    UseLoopCountLimit=false
   ) {module=SpawnBurst_Instantaneous001}
  ]
  particleSpawnScript [
   &InitializeParticle(
    Color Maximum=<rgba 12 0.3413 0 1>
    Color Minimum=<rgba 7 0.0668 0.2674 1>
    Color Mode=::NewEnumerator2
    Lifetime=&UniformRangedFloat(Minimum=0.125, Maximum=0.2)
    Mass Mode=::NewEnumerator1
    Position Offset=$System.OrbWPO
    Position Offset Coordinate Space=::World
    Sprite Rotation Mode=::NewEnumerator2
    Sprite Size=&UniformRangedVector2D(Minimum=<250 250>, Maximum=<400 400>)
    Sprite Size Mode=::NewEnumerator4
    UsePositionOffset=true
    Write Material Random=true
   ) {module=InitializeParticle001}
   &DynamicMaterialParameters(
    Index 0 Param 1=$Particles.MaterialRandom
    Index 0 Param 2=&Lerp_Float(A=0.75, B=1.2, Alpha=$Particles.MaterialRandom)
    Index 0 Param 4=1
    Module.Param0WriteEnabled=true
    Module.Param1WriteEnabled=true
    Module.Param2WriteEnabled=false
    Module.Param3WriteEnabled=true
   )
   &CameraOffset(Camera Offset Amount=25)
   &ShapeLocation(
    MaskSpawnGroup=false
    Override Seed=false
    Ring Radius=250
    Rotation Mode=::NewEnumerator1
    Shape Primitive=::NewEnumerator4
    UseOwnerScale=true
   )
   &AddVelocity(
    Module.Use Falloff Distance=false
    Origin Offset=<-200 0 0>
    Override Seed=false
    Velocity Mode=::NewEnumerator1
    Velocity Origin=&AddVectorToPosition(Position=$Engine.Emitter.SimulationPosition, Vector=$System.OrbWPO)
    Velocity Speed=500
   )
  ]
  particleUpdateScript [
   &ParticleState()
   &DynamicMaterialParameters(
    Index 0 Param 3=&FloatFromCurve(FloatCurve=DataInterface() {propertyValues {Curve {keys [<at 0.5 0 RCIM_Cubic>, {interpMode=RCIM_Cubic, leaveTangent=3.7979, time=1, value=1}]}}})
    Module.Param0WriteEnabled=false
    Module.Param1WriteEnabled=false
    Module.Param2WriteEnabled=true
    Module.Param3WriteEnabled=false
   ) {module=DynamicMaterialParameters001}
   &ScaleSpriteSize(Uniform Curve Sprite Scale=DataInterface() {propertyValues {Curve {keys [
        {arriveTangent=0.5666, interpMode=RCIM_Cubic, leaveTangent=0.5666, tangentMode=RCTM_User, time=0, value=0.75}
        {tangentMode=RCTM_User, time=1, value=1}
       ]}}})
   &ScaleColor(
    Scale RGB=&MakeVectorFromLinearColorRGB(Color=&ColorFromCurve(ColorCurve=DataInterface() {propertyValues {
        AlphaCurve {keys []}
        BlueCurve {keys [<at 0.0806 2.7936 RCIM_Cubic>,<at 0.4355 1.5>]}
        GreenCurve {keys [<at 0.0806 1.9309 RCIM_Cubic>,<at 0.4355 0.0849>]}
        RedCurve {keys [<at 0.0806 4 RCIM_Cubic>,<at 0.4355 1.2907>]}
       }}))
    ScaleA=false
    ScaleRGB=true
   ) {module=ScaleColor001}
   &SolveForcesAndVelocity(Clamp Velocity=false, Limit Acceleration=false)
  ]
  rendererClasses [&NiagaraSpriteRendererProperties]
  renderers [{rendererClass=&NiagaraSpriteRendererProperties}]
 }
 FlareLight {
  emitterUpdateScript [
   &EmitterState(
    Inactive Response=::NewEnumerator1
    Life Cycle Mode=::NewEnumerator1
    Loop Behavior=::NewEnumerator1
    Loop Duration Mode=::NewEnumerator1
    UseLoopDelay=false
   )
   &SpawnBurst_Instantaneous(Use Spawn Probability=false, UseLoopCountLimit=false)
  ]
  particleSpawnScript [
   &InitializeParticle(
    Color=<rgba 1 0.0382 0.0095 1>
    Sprite Size Mode=::NewEnumerator3
    Uniform Sprite Size=1400
    UsePositionOffset=false
    Write Material Random=true
   )
   &CameraOffset(Camera Offset Amount=500)
   &SpriteFacingAndAlignment(WriteAlignment=false, WriteFacing=true)
   /SetVariables(Particles.DarkenColor:LinearColor=<rgba 0.016 0.0041 0 1.25>) {isSetParameters=true}
   &DynamicMaterialParameters(
    Index 0 Param 1=3
    Index 0 Param 2=1
    Index 3 Param 1=1
    Index 3 Param 2=0.5
    Module.Index0Float0Write=true
    Module.Index0Float1Write=true
    Module.Index0Float2Write=false
    Module.Index0Float3Write=false
    Module.Index3Float0Write=true
    Module.Index3Float1Write=true
    Module.Index3Float2Write=false
    Module.Index3Float3Write=false
    Write Parameter Index 3=::NewEnumerator1
   )
  ]
  particleUpdateScript [
   &ParticleState(Kill Particles When Lifetime Has Elapsed=false)
   &ScaleColor(
    Scale Alpha=&FloatFromCurve(FloatCurve=DataInterface() {propertyValues {Curve {
        defaultValue=1
        keys [
         {arriveTangent=13.1896, interpMode=RCIM_Cubic, leaveTangent=13.1896, tangentMode=RCTM_User, time=0, value=0}
         {interpMode=RCIM_Cubic, leaveTangent=-2.8238, time=0.15, value=1}
        ]
       }}})
    ScaleA=true
    ScaleRGB=false
   )
   &ScaleSpriteSize(Uniform Curve Sprite Scale=DataInterface() {propertyValues {Curve {keys [
        {arriveTangent=7.8051, interpMode=RCIM_Cubic, time=0, value=0}
        {arriveTangent=-1.4366, interpMode=RCIM_Cubic, leaveTangent=-1.4366, tangentMode=RCTM_User, time=0.2074, value=1.5}
        {tangentMode=RCTM_User, time=1, value=1}
       ]}}})
   /SetVariables(Particles.Position=&AddVectorToPosition(
     Position=<0 0 0>
     Vector=&TransformVector(Vector=$System.OrbWPO, Source Space=::World, Destination Space=::Local)
    )) {isSetParameters=true}
   /SetVariables(Particles.LightPosition:NiagaraPosition=&AddVectorToPosition(Position=$Particles.Position, Vector=<500 0 0>)) {isSetParameters=true}
   /SetVariables(Particles.LightColor=&Multiply_LinearColorByFloat(Linear Color=$Particles.Color, Float=10)) {isSetParameters=true}
   /SetVariables(Particles.LightRadius=&FloatFromCurve(
     FloatCurve=DataInterface() {propertyValues {Curve {
        defaultValue=1
        keys [
         {arriveTangent=1.8849, interpMode=RCIM_Cubic, leaveTangent=1.8849, tangentMode=RCTM_User, time=0, value=0.3}
         {arriveTangent=0.2703, tangentMode=RCTM_User, time=1, value=1}
        ]
       }}}
     Scale Curve=2000
    )) {isSetParameters=true}
  ]
  rendererClasses [&NiagaraSpriteRendererProperties, &NiagaraLightRendererProperties]
  renderers [
   {rendererClass=&NiagaraSpriteRendererProperties}
   {rendererClass=&NiagaraLightRendererProperties, rendererIndex=1}
   {rendererClass=&NiagaraSpriteRendererProperties, rendererIndex=2}
  ]
 }
 Minimal {
  emitterUpdateScript [
   &EmitterState(
    Inactive Response=::NewEnumerator1
    Life Cycle Mode=::NewEnumerator1
    Loop Behavior=::NewEnumerator1
    Loop Duration Mode=::NewEnumerator1
    UseLoopDelay=false
   )
   &SpawnBurst_Instantaneous(Spawn Count=4, Use Spawn Probability=false, UseLoopCountLimit=false)
   &SpawnBurst_Instantaneous(Spawn Count=4, Spawn Group=1, Use Spawn Probability=false, UseLoopCountLimit=false) {module=SpawnBurst_Instantaneous001}
   &SpawnBurst_Instantaneous(Spawn Count=4, Spawn Group=2, Use Spawn Probability=false, UseLoopCountLimit=false) {module=SpawnBurst_Instantaneous002}
  ]
  particleSpawnScript [
   &PartitionParticles(Number Of Partitions=4)
   &InitializeParticle(
    Color=&MaskLinearColorBySpawnGroup(Masked Linear Color=<rgba 10 0 0.247 1>, Passthrough Linear Color=<rgba 8 0 0.1976 1>)
    Lifetime=5
    Mesh Renderer Array Visibility Mode=::NewEnumerator1
    Mesh Renderer Info=DataInterface() {propertyValues {MeshRenderer=&self:Minimal.NiagaraMeshRendererProperties_0}}
    Mesh Renderer Mesh Index:NiagaraInt32=&SpawnGroup()
    Mesh Scale Mode=::NewEnumerator3
    Mesh Uniform Scale=4.8
    UsePositionOffset=false
    Write Material Random=true
   )
   &InitialMeshOrientation(
    Mesh Axis To Orient=<0 0.01 1>
    Mesh Orientation Mode=::NewEnumerator2
    Orientation Vector=&RotateVectorByQuaternion(
     Vector To Rotate=<0 0 1>
     Quaternion=&MakeQuaternion(
      Quaternion From=0
      Axis=<1 0 0>
      Angle=&Add_Float(
       A=$Particles.PartitionParticles.NormalizedPartition
       B=&MaskFloatBySpawnGroup(Masked Float=0.125, Passthrough Float=0)
      )
      From=?
      To=?
      Angle Type=::NewEnumerator1
      Normalize Vectors=true
     )
    )
    Use Rotation Vector=false
   ) {enabled=false}
   &CameraOffset(Camera Offset Amount=150)
  ]
  particleUpdateScript [
   &ParticleState(Kill Particles When Lifetime Has Elapsed=false)
   /SetVariables(Particles.MeshOrientation=&MakeQuaternion(
     Quaternion From=0
     Axis=<1 0 0>
     Angle=&Add_Float(
      A=$Particles.PartitionParticles.NormalizedPartition
      B=&MaskFloatBySpawnGroup(
       Masked Float=0.125
       Passthrough Float=&FloatFromCurve(
        FloatCurve=DataInterface() {propertyValues {Curve {
           defaultValue=1
           keys [
            {interpMode=RCIM_Cubic, tangentMode=RCTM_User, time=0.3, value=0}
            {arriveTangentWeight=0.0833, interpMode=RCIM_Cubic, leaveTangentWeight=0.1323, tangentMode=RCTM_User, time=0.4, value=1}
            {arriveTangentWeight=0.2205, interpMode=RCIM_Cubic, leaveTangentWeight=0.1961, tangentMode=RCTM_User, time=1.9, value=1}
            {tangentMode=RCTM_User, time=2, value=0}
           ]
          }}}
        CurveIndex=&ModuloFloat(A=&Multiply_Float(A=$Emitter.Age, B=0.3), B=3)
        Scale Curve=0.25
       )
      )
     )
     From=?
     To=?
     Angle Type=::NewEnumerator1
     Normalize Vectors=false
    )) {isSetParameters=true}
   /SetVariables(Particles.Position=&AddVectorToPosition(
     Position=&Add_Vector(
      A=&Multiply_VectorByFloat(
       Vector=&ReturnMeshOrientationAxis(Mesh Orientation Axis=::ZAxis, Mesh Orientation Quat=$Particles.MeshOrientation)
       Float=&FloatFromCurve(
        FloatCurve=DataInterface() {propertyValues {Curve {
           defaultValue=1
           keys [
            {interpMode=RCIM_Cubic, tangentMode=RCTM_User, time=0.5, value=0.25}
            {arriveTangentWeight=0.0833, interpMode=RCIM_Cubic, leaveTangentWeight=0.1323, tangentMode=RCTM_User, time=0.55, value=1}
            {arriveTangentWeight=0.2205, interpMode=RCIM_Cubic, leaveTangentWeight=0.1961, tangentMode=RCTM_User, time=0.95, value=1}
            {tangentMode=RCTM_User, time=1, value=0.25}
           ]
          }}}
        CurveIndex=&ModuloFloat(
         A=&Multiply_Float(A=$Emitter.Age, B=0.25)
         B=&MaskFloatBySpawnGroup(Masked Float=1.5, Passthrough Float=1)
        )
        Scale Curve=&MaskFloatBySpawnGroup(Masked Float=40, Passthrough Float=80)
       )
      )
      B=&Multiply_VectorByFloat(
       Vector=&ReturnMeshOrientationAxis(Mesh Orientation Axis=::YAxis, Mesh Orientation Quat=$Particles.MeshOrientation)
       Float=&FloatFromCurve(
        FloatCurve=DataInterface() {propertyValues {Curve {
           defaultValue=1
           keys [
            {interpMode=RCIM_Cubic, tangentMode=RCTM_User, time=0, value=0}
            {arriveTangentWeight=0.0833, interpMode=RCIM_Cubic, leaveTangentWeight=0.1323, tangentMode=RCTM_User, time=0.1, value=1}
            {arriveTangentWeight=0.2205, interpMode=RCIM_Cubic, leaveTangentWeight=0.1961, tangentMode=RCTM_User, time=0.6, value=1}
            {tangentMode=RCTM_User, time=0.7, value=0}
           ]
          }}}
        CurveIndex=&ModuloFloat(A=&Multiply_Float(A=$Emitter.Age, B=0.5), B=1.75)
        Scale Curve=&SelectFloatFromArray(Direct Array Index=&SpawnGroup(), Float Selection Array=DataInterface() {propertyValues {FloatData [0,25,-25]}})
       )
      )
     )
     Vector=&TransformVector(Vector=$System.OrbWPO, Source Space=::World, Destination Space=::Local)
    )) {isSetParameters=true}
  ]
  rendererClasses [&NiagaraSpriteRendererProperties, &NiagaraMeshRendererProperties]
  renderers [
   {rendererClass=&NiagaraSpriteRendererProperties}
   {rendererClass=&NiagaraMeshRendererProperties, rendererIndex=1}
  ]
 }
 TargetingBeam001 {
  bEnabled=false
  emitterUpdateScript [
   &EmitterState(
    Inactive Response=::NewEnumerator1
    Life Cycle Mode=::NewEnumerator1
    Loop Behavior=::NewEnumerator1
    Loop Duration Mode=::NewEnumerator1
    UseLoopDelay=false
   )
   &SpawnBurst_Instantaneous(Use Spawn Probability=false, UseLoopCountLimit=false)
  ]
  particleSpawnScript [
   &InitializeParticle(
    Color=<rgba 2.5 0 0 1>
    Mesh Scale Mode=::NewEnumerator3
    Mesh Uniform Scale=0.6
    Sprite Size Mode=::NewEnumerator3
    Uniform Sprite Size=200
    UsePositionOffset=false
    Write Material Random=true
   )
   &CameraOffset(Camera Offset Amount=100)
   &DynamicMaterialParameters(
    Index 0 Param 1=1
    Index 0 Param 2=100
    Index 0 Param 3=1
    Module.Index0Float0Write=true
    Module.Index0Float1Write=true
    Module.Index0Float2Write=true
    Module.Index0Float3Write=true
   ) {module=DynamicMaterialParameters001}
  ]
  particleUpdateScript [
   &ParticleState(Kill Particles When Lifetime Has Elapsed=false)
   /SetVariables(Particles.VisibilityTag=&MakeCustomIntFromBool(Bool=$User.Targeting, True INT=1, False INT=0)) {isSetParameters=true}
   /SetVariables(Particles.Position=&AddVectorToPosition(Position=$Engine.Owner.Position, Vector=$System.OrbWPO)) {isSetParameters=true}
   /SetVariables(Particles.Oscillation:Vector3f=&RotateVectorByQuaternion(
     Vector To Rotate=&Multiply_VectorByFloat(Vector=&Add_Vector(A=$Engine.Owner.SystemXAxis, B=<0 0 -0.4>), Float=7000)
     Quaternion=&MakeQuaternion(
      Quaternion From=0
      Axis=$Engine.Owner.SystemZAxis
      Angle=&Waveform(Global Amplitude Scale=45, "Waveform [1]"=::NewEnumerator0, "[1] Frequency"=0.2, "[1] Amplitude Min/Max"=<-1 1>)
      From=?
      To=?
      Angle Type=::NewEnumerator0
      Normalize Vectors=true
     )
    )) {isSetParameters=true}
   &InterpolateOverTime(
    Rate Of Change=4
    Target Position Value:NiagaraPosition=&AddVectorToPosition(Position=$Particles.Position, Vector=$Particles.Oscillation)
    Variable Type=::NewEnumerator6
   )
   /SetVariables(Particles.EndPointPosition:NiagaraPosition=$Particles.InterpolateOverTime.InterpolateOverTime_Position.Moving Average) {isSetParameters=true}
   /SetVariables(Particles.BeamPosition:NiagaraPosition=&Lerp_Vector(
     A=Linked() {linkedVariable {name=Particles.Position, type=NiagaraPosition}}
     B=Linked() {linkedVariable {name=Particles.EndPointPosition, type=NiagaraPosition}}
     Alpha=0.5
    )) {isSetParameters=true}
   /SetVariables(Particles.BeamSpriteSize:Vector2f=&MakeVector2D(
     X=&FN_LOD_Distance_Remap(Start Distance=2000, End Distance=10000, Minimum Output=10, Maximum Output=50, Clamp Output=true)
     Y=&Multiply_Float(A=&DistanceBetweenPositions(A=$Particles.EndPointPosition, B=$Particles.Position), B=1)
    )) {isSetParameters=true}
   &SpriteFacingAndAlignment(
    Sprite Alignment=&Subtract_Position(A=$Particles.Position, B=$Particles.EndPointPosition)
    Sprite Facing=&ReturnMeshOrientationAxis(Mesh Orientation Axis=::XAxis, Mesh Orientation Quat=$Particles.MeshOrientation)
    WriteAlignment=true
    WriteFacing=true
   )
   /SetVariables(Particles.Sprite Start Camera Offset=300) {isSetParameters=true}
  ]
  rendererClasses [&NiagaraSpriteRendererProperties]
  renderers [
   {rendererClass=&NiagaraSpriteRendererProperties}
   {rendererClass=&NiagaraSpriteRendererProperties, rendererIndex=1}
  ]
 }
}
summary {
 systemName=NS_DragonCart_Orb_Aggressive
 emitters [
  {emitterName=Minimal, rendererClasses [&NiagaraSpriteRendererProperties, &NiagaraMeshRendererProperties]}
  {emitterName=FlareLight, rendererClasses [&NiagaraSpriteRendererProperties, &NiagaraLightRendererProperties]}
  {emitterName=ElectricArcs, rendererClasses [&NiagaraSpriteRendererProperties]}
  {bEnabled=false, emitterName=TargetingBeam001, rendererClasses [&NiagaraSpriteRendererProperties]}
 ]
 userVariables [User.Targeting {type=NiagaraBool}]
}
